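# Physics asset for the demo skeletal mesh of the Football Trophy pickaxe, a two-handed melee cosmetic weapon.
# It holds a single capsule body on the Melee_TwoHanded_Handle bone, offset along Z from the bone origin and pitched to a fixed orientation.
# There are no constraints, so the weapon collides as one rigid shape.
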PhysicsAsset /BRCosmetics/Weapons/FORT_Melee/Pickaxe_FootballTrophy/Meshes/Demo/SK_Pickaxe_FootballTrpohy_Demo_PhysicsAsset
bodies {Melee_TwoHanded_Handle [{shapeType=Capsule, length=91.1463, radius=29.2985, center=<0.1318 -0.4444 32.9976>, rotation=<rot 0.0184 -90 0.008>}]}
constraints []
